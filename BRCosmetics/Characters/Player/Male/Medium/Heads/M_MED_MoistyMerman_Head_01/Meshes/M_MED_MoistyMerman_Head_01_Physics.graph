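# Physics asset for the Moisty Merman head of a medium male player cosmetic, providing physics bodies for its dangling tongue.
# The head and two tongue bones each carry a zero-length capsule, effectively a sphere, with the head's being larger.
# Two constraints chain the tongue tip to the tongue base and the base to the head, locking twist and defining swing angles per joint.
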PhysicsAsset /BRCosmetics/Characters/Player/Male/Medium/Heads/M_MED_MoistyMerman_Head_01/Meshes/M_MED_MoistyMerman_Head_01_Physics
bodies {
 dyn_tongue_01 [{shapeType=Capsule, length=0, radius=2.5, center=<-1 0 0>}]
 dyn_tongue_02 [{shapeType=Capsule, length=0, radius=2.5, center=<-1 0 0>}]
 head [{shapeType=Capsule, length=0, radius=5}]
}
constraints [
 {bone1Name=dyn_tongue_01, bone2Name=head, swing1LimitDegrees=51.8571, swing2LimitDegrees=90, twistLimitDegrees=17.5714, twistMotion=Locked}
 {
  bone1Name=dyn_tongue_02
  bone2Name=dyn_tongue_01
  swing1LimitDegrees=45
  swing2LimitDegrees=48.8571
  twistLimitDegrees=45
  twistMotion=Locked
 }
]
