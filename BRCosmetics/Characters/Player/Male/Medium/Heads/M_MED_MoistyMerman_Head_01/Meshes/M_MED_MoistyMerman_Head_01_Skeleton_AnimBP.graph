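# Post-animation Blueprint for the Moisty Merman player head, adding secondary motion to its fins, tongue and spit bones.
# It copies the owner mesh's pose, runs Anim Dynamics on the fins and spit, then caches the result.
# That pose feeds either a RigidBody node or, when rigid body is disabled or LOD exceeds 0, a rotated tongue with Anim Dynamics.
# Targeting events set Ads_Bool, which the graphs never read.
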AnimBlueprint /BRCosmetics/Characters/Player/Male/Medium/Heads/M_MED_MoistyMerman_Head_01/Meshes/M_MED_MoistyMerman_Head_01_Skeleton_AnimBP
target_skeleton=&/M_MED_MoistyMerman_Head_01_Skeleton
graphs [
 AnimGraph {nodes [
   n1 = RigidBody(ComponentPose=Local To Component(LocalPose=Use cached pose 'noRBAN'())) {
    lODThreshold=0
    overridePhysicsAsset=&/M_MED_MoistyMerman_Head_01_Physics
    alphaBoolBlend {blendOption=Linear}
   }
   n2 = "Transform (Modify) Bone - Bone: dyn_tongue_01"!(
    ComponentPose=Local To Component(LocalPose=Use cached pose 'noRBAN'())
    bAlphaBoolEnabled:bool=True
    Translation="0,0,0"
    Rotation="70,0,-10"
    Scale="1,1,1"
   ) -> (Pose) {
    alphaInputType=Bool
    rotationMode=BMM_Additive
    rotationSpace=BCS_BoneSpace
    alphaBoolBlend {blendOption=Linear}
    boneToModify {boneName=dyn_tongue_01}
    rotation {pitch=70, roll=-10}
   }
   n3 = Anim Dynamics - Bone: dyn_tongue_02(ComponentPose=@n2, Alpha=1, ExternalForce="0,0,0") {
    angularDampingOverride=0.9999
    angularSpringConstant=1000
    bAngularSpring=true
    bGravityOverrideInSimSpace=true
    bOverrideAngularDamping=true
    lODThreshold=1
    numSolverIterationsPostUpdate=4
    numSolverIterationsPreUpdate=16
    simulationSpace=RootRelative
    alphaBoolBlend {blendOption=Linear}
    boundBone {boneName=dyn_tongue_02}
    componentAppliedLinearAccClamp=<3000 3000 3000>
    physicsBodyDefinitions [{
      sphereCollisionRadius=0
      boundBone {boneName=dyn_tongue_02}
      boxExtents=<15 15 15>
      constraintSetup {angularTargetAxis=AxisZ, angularLimitsMax=<0 30 30>, angularLimitsMin=<0 -15 -30>, angularTarget=<0 0 0>}
      localJointOffset=<-1 -1 0>
     }]
    relativeSpaceBone {boneName=head}
    simSpaceSettings {simSpaceAngularAlpha=0.3}
   }
   n4 = Anim Dynamics - Bone: dyn_tongue_01(ComponentPose=@n3, Alpha=1, ExternalForce="0,0,0") {
    angularDampingOverride=0.9999
    angularSpringConstant=1000
    bAngularSpring=true
    bGravityOverrideInSimSpace=true
    bOverrideAngularDamping=true
    lODThreshold=1
    numSolverIterationsPostUpdate=4
    numSolverIterationsPreUpdate=16
    simulationSpace=RootRelative
    alphaBoolBlend {blendOption=Linear}
    boundBone {boneName=dyn_tongue_01}
    componentAppliedLinearAccClamp=<3000 3000 3000>
    physicsBodyDefinitions [{
      sphereCollisionRadius=0
      boundBone {boneName=dyn_tongue_01}
      boxExtents=<15 15 15>
      constraintSetup {angularTargetAxis=AxisZ, angularLimitsMax=<0 35 40>, angularLimitsMin=<0 0 -40>, angularTarget=<0 0 0>}
      localJointOffset=<-2 1 0>
     }]
    relativeSpaceBone {boneName=head}
    simSpaceSettings {simSpaceAngularAlpha=0.3}
   }
   Output Pose(Result=Blend Poses by bool(
     BlendPose_0=Component To Local(ComponentPose=@n4)
     BlendPose_1=Component To Local(ComponentPose=@n1)
     bActiveValue=Get isRigidBodyDisabled() -> (isRigidBodyDisabled:bool)
    )) {layerGroup=None}
   n5 = Get MeshToCopy() -> (MeshToCopy:Skeletal Mesh Component Object Reference)
   n6 = Anim Dynamics - Bone: dyn_headFin_l(ComponentPose=Local To Component(LocalPose=Copy Pose From Mesh(SourceMeshComponent=@n5)), Alpha=1, ExternalForce="0,0,0") {
    alpha=0.9
    angularDampingOverride=0.9999
    angularSpringConstant=1500
    bAngularSpring=true
    bGravityOverrideInSimSpace=true
    bOverrideAngularDamping=true
    numSolverIterationsPostUpdate=4
    numSolverIterationsPreUpdate=16
    simulationSpace=RootRelative
    alphaBoolBlend {blendOption=Linear}
    boundBone {boneName=dyn_headFin_l}
    physicsBodyDefinitions [{
      sphereCollisionRadius=0
      boundBone {boneName=dyn_headFin_l}
      boxExtents=<15 5 5>
      constraintSetup {angularTargetAxis=AxisZ, twistAxis=AxisY, angularLimitsMax=<10 10 30>, angularLimitsMin=<-10 -10 -30>, angularTarget=<0 0 1>}
      localJointOffset=<0 0 0>
     }]
    simSpaceSettings {simSpaceAngularAlpha=0.1}
   }
   n7 = Anim Dynamics - Bone: dyn_headFin_r(ComponentPose=@n6, Alpha=1, ExternalForce="0,0,0") {
    alpha=0.9
    angularDampingOverride=0.9999
    angularSpringConstant=1500
    bAngularSpring=true
    bGravityOverrideInSimSpace=true
    bOverrideAngularDamping=true
    numSolverIterationsPostUpdate=4
    numSolverIterationsPreUpdate=16
    simulationSpace=RootRelative
    alphaBoolBlend {blendOption=Linear}
    boundBone {boneName=dyn_headFin_r}
    physicsBodyDefinitions [{
      sphereCollisionRadius=0
      boundBone {boneName=dyn_headFin_r}
      boxExtents=<15 5 5>
      constraintSetup {angularTargetAxis=AxisZ, twistAxis=AxisY, angularLimitsMax=<10 10 30>, angularLimitsMin=<-10 -10 -30>, angularTarget=<0 0 1>}
      localJointOffset=<0 0 0>
     }]
    simSpaceSettings {simSpaceAngularAlpha=0.1}
   }
   n8 = Anim Dynamics - Bone: dyn_headFin_c(ComponentPose=@n7, Alpha=0.3, ExternalForce="0,0,0") {
    alpha=0.9
    angularDampingOverride=0.9999
    angularSpringConstant=1500
    bAngularSpring=true
    bGravityOverrideInSimSpace=true
    bOverrideAngularDamping=true
    numSolverIterationsPostUpdate=4
    numSolverIterationsPreUpdate=16
    simulationSpace=RootRelative
    alphaBoolBlend {blendOption=Linear}
    boundBone {boneName=dyn_headFin_c}
    physicsBodyDefinitions [{
      sphereCollisionRadius=0
      boundBone {boneName=dyn_headFin_c}
      boxExtents=<15 5 5>
      constraintSetup {angularTargetAxis=AxisZ, twistAxis=AxisY, angularLimitsMax=<10 10 30>, angularLimitsMin=<-10 -10 -30>, angularTarget=<0 0 1>}
      localJointOffset=<0 0 0>
     }]
    simSpaceSettings {simSpaceAngularAlpha=0.1}
   }
   n9 = Anim Dynamics - Bone: dyn_Spit(ComponentPose=@n8, Alpha=1, ExternalForce="0,0,0") {
    angularDampingOverride=0.9999
    bGravityOverrideInSimSpace=true
    bOverrideAngularDamping=true
    lODThreshold=1
    numSolverIterationsPostUpdate=4
    numSolverIterationsPreUpdate=16
    simulationSpace=RootRelative
    alphaBoolBlend {blendOption=Linear}
    boundBone {boneName=dyn_Spit}
    physicsBodyDefinitions [{
      sphereCollisionRadius=0
      boundBone {boneName=dyn_Spit}
      boxExtents=<5 5 5>
      constraintSetup {angularTargetAxis=AxisZ, twistAxis=AxisZ, angularLimitsMax=<5 30 30>, angularLimitsMin=<-90 -30 -30>, angularTarget=<0 0 0>}
      localJointOffset=<0 0 -0.5>
     }]
    simSpaceSettings {simSpaceAngularAlpha=0.1}
   }
   Save cached pose 'noRBAN'(Pose=Component To Local(ComponentPose=@n9)) {cacheName=noRBAN}
  ]}
 EventGraph {nodes [
   Event BlueprintUpdateAnimation()
   n1 = Get CurrentLOD()
   n2 = Get bRigidBodyEnabled()
   > Set isRigidBodyDisabled(isRigidBodyDisabled:bool=OR Boolean(A=NOT Boolean(A=@n2), B="integer > integer"(A=@n1, B=0) {is=title K2Node_CallFunction})) -> (Output_Get:bool)
   Event OnBeginTargeting() {is=title K2Node_Event}
   > Set Ads_Bool(Ads_Bool:bool=true) -> (Output_Get:bool)
   Event OnEndTargeting() {is=title K2Node_Event}
   > Set Ads_Bool(Ads_Bool=false)
   Event BlueprintInitializeAnimation()
   > Set MeshToCopy(MeshToCopy:Skeletal Mesh Component Object Reference=GetOwnerSkeletalMeshComponent()) -> (Output_Get:Skeletal Mesh Component Object Reference)
   TryGetPawnOwner()
  ]}
]
variables [
 Ads_Bool {type=boolean}
 isRigidBodyDisabled {type=boolean}
 MeshToCopy {type=SkeletalMeshComponent}
]
